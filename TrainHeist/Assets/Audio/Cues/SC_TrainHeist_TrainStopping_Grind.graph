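# Sound cue for the grinding sound of a stopping train in Train Heist, played in the gameplay events sound class with its own attenuation settings.
# A Modulator feeds a quality-level switch that picks one of two equally weighted Random nodes.
# The first two quality slots draw from all nine grind waves and the last two from only the first four.
# The nine-wave Random node is excluded from branch culling.
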SoundCue /TrainHeist/Assets/Audio/Cues/SC_TrainHeist_TrainStopping_Grind
root=Modulator_0
nodes [
 {id=Modulator_0, children [QualityLevel_0]}
 {id=QualityLevel_0, children [Random_0,Random_0,Random_1,Random_1]}
 {
  id=Random_0
  children [WavePlayer_0,WavePlayer_1,WavePlayer_2,WavePlayer_3,WavePlayer_4,WavePlayer_5,WavePlayer_6,WavePlayer_7,WavePlayer_8]
  properties {bShouldExcludeFromBranchCulling=true, weights [1,1,1,1,1,1,1,1,1]}
 }
 {id=Random_1, children [WavePlayer_9, WavePlayer_10, WavePlayer_11, WavePlayer_12], properties {weights [1,1,1,1]}}
 {id=WavePlayer_0, properties {soundWaveAssetPtr=/TrainHeist/Assets/Audio/Waves/TrainHeist_TrainStopping_Grind_01}}
 {id=WavePlayer_1, properties {soundWaveAssetPtr=/TrainHeist/Assets/Audio/Waves/TrainHeist_TrainStopping_Grind_02}}
 {id=WavePlayer_10, properties {soundWaveAssetPtr=&TrainHeist_TrainStopping_Grind_02}}
 {id=WavePlayer_11, properties {soundWaveAssetPtr=/TrainHeist/Assets/Audio/Waves/TrainHeist_TrainStopping_Grind_03}}
 {id=WavePlayer_12, properties {soundWaveAssetPtr=/TrainHeist/Assets/Audio/Waves/TrainHeist_TrainStopping_Grind_04}}
 {id=WavePlayer_2, properties {soundWaveAssetPtr=&TrainHeist_TrainStopping_Grind_03}}
 {id=WavePlayer_3, properties {soundWaveAssetPtr=&TrainHeist_TrainStopping_Grind_04}}
 {id=WavePlayer_4, properties {soundWaveAssetPtr=/TrainHeist/Assets/Audio/Waves/TrainHeist_TrainStopping_Grind_05}}
 {id=WavePlayer_5, properties {soundWaveAssetPtr=/TrainHeist/Assets/Audio/Waves/TrainHeist_TrainStopping_Grind_06}}
 {id=WavePlayer_6, properties {soundWaveAssetPtr=/TrainHeist/Assets/Audio/Waves/TrainHeist_TrainStopping_Grind_07}}
 {id=WavePlayer_7, properties {soundWaveAssetPtr=/TrainHeist/Assets/Audio/Waves/TrainHeist_TrainStopping_Grind_08}}
 {id=WavePlayer_8, properties {soundWaveAssetPtr=/TrainHeist/Assets/Audio/Waves/TrainHeist_TrainStopping_Grind_09}}
 {id=WavePlayer_9, properties {soundWaveAssetPtr=&TrainHeist_TrainStopping_Grind_01}}
]
properties {
 attenuationSettings=/TrainHeist/Assets/Audio/Settings/ATT_TrainHeist_TainStopping
 maxDistance=18500
 soundClassObject=&Fort_Events_Gameplay
 volumeMultiplier=1
}
