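# Sound cue for the train alarm in TrainHeist, playing a start sound that leads into a looping alarm.
# A root delay holds back the whole cue, then a mixer plays the louder one-shot start wave alongside a further-delayed looping wave.
# Fixed delays stagger the intro and loop in a single cue, spatialized with the train-stopping attenuation and routed to the gameplay events sound class.
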SoundCue /TrainHeist/Assets/Audio/Cues/SC_TrainHeist_Train_Alarm
root=Delay_1
nodes [
 {id=Delay_0, children [WavePlayer_0], properties {delayMax=2, delayMin=2}}
 {id=Delay_1, children [Mixer_0], properties {delayMax=3.2, delayMin=3.2}}
 {id=Mixer_0, children [Delay_0,WavePlayer_1], properties {inputVolume [1,2]}}
 {id=WavePlayer_0, properties {bLooping=true, soundWaveAssetPtr=/TrainHeist/Assets/Audio/Waves/TrainHeist_Alarm_Loop}}
 {id=WavePlayer_1, properties {soundWaveAssetPtr=/TrainHeist/Assets/Audio/Waves/TrainHeist_Alarm_Start}}
]
properties {
 attenuationSettings=/TrainHeist/Assets/Audio/Settings/ATT_TrainHeist_TainStopping
 maxDistance=18500
 soundClassObject=&Fort_Events_Gameplay
 volumeMultiplier=0.4
}
