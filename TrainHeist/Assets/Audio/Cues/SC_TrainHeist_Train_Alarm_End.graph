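# Sound cue for the end of the train alarm in the TrainHeist content.
# A single wave player plays the TrainHeist_Alarm_End wave with no randomization or mixing.
# It uses the train-stopping attenuation asset for spatial falloff, plays at reduced volume, and routes through the gameplay events sound class.
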SoundCue /TrainHeist/Assets/Audio/Cues/SC_TrainHeist_Train_Alarm_End
root=WavePlayer_2
nodes [{id=WavePlayer_2, properties {soundWaveAssetPtr=/TrainHeist/Assets/Audio/Waves/TrainHeist_Alarm_End}}]
properties {
 attenuationSettings=/TrainHeist/Assets/Audio/Settings/ATT_TrainHeist_TainStopping
 maxDistance=18500
 soundClassObject=&Fort_Events_Gameplay
 volumeMultiplier=0.5
}
