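# Sound cue for the Train Heist train explosion, playing a randomly chosen car explosion wave.
# A Quality Level node gives the first two levels an equally weighted pool of six waves and the last two a pool of the first three.
# The pick passes through a default Modulator for variation.
# The cue uses the large-explosives sound class and a train-stopping attenuation asset.
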SoundCue /TrainHeist/Assets/Audio/Cues/SC_TrainHeist_Train_Explosion
root=Modulator_0
nodes [
 {id=Modulator_0, children [QualityLevel_0]}
 {id=QualityLevel_0, children [Random_2,Random_2,Random_3,Random_3]}
 {
  id=Random_2
  children [WavePlayer_13, WavePlayer_14, WavePlayer_15, WavePlayer_16, WavePlayer_17, WavePlayer_18]
  properties {bShouldExcludeFromBranchCulling=true, weights [1,1,1,1,1,1]}
 }
 {id=Random_3, children [WavePlayer_19, WavePlayer_20, WavePlayer_21], properties {weights [1,1,1]}}
 {id=WavePlayer_13, properties {soundWaveAssetPtr=/TrainHeist/Assets/Audio/Waves/TrainHeist_Car_Explosion_01}}
 {id=WavePlayer_14, properties {soundWaveAssetPtr=/TrainHeist/Assets/Audio/Waves/TrainHeist_Car_Explosion_02}}
 {id=WavePlayer_15, properties {soundWaveAssetPtr=/TrainHeist/Assets/Audio/Waves/TrainHeist_Car_Explosion_03}}
 {id=WavePlayer_16, properties {soundWaveAssetPtr=/TrainHeist/Assets/Audio/Waves/TrainHeist_Car_Explosion_04}}
 {id=WavePlayer_17, properties {soundWaveAssetPtr=/TrainHeist/Assets/Audio/Waves/TrainHeist_Car_Explosion_05}}
 {id=WavePlayer_18, properties {soundWaveAssetPtr=/TrainHeist/Assets/Audio/Waves/TrainHeist_Car_Explosion_06}}
 {id=WavePlayer_19, properties {soundWaveAssetPtr=&TrainHeist_Car_Explosion_01}}
 {id=WavePlayer_20, properties {soundWaveAssetPtr=&TrainHeist_Car_Explosion_02}}
 {id=WavePlayer_21, properties {soundWaveAssetPtr=&TrainHeist_Car_Explosion_03}}
]
properties {
 attenuationSettings=/TrainHeist/Assets/Audio/Settings/ATT_TrainHeist_TainStopping
 maxDistance=18500
 soundClassObject=/Game/Sounds/FortSoundClassesAndModes/Fort_Explosives_Large_PSM
 volumeMultiplier=0.7
}
